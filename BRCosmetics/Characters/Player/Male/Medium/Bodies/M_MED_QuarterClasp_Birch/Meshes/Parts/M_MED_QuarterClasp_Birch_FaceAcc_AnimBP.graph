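# Post-process-style AnimBP for the Birch QuarterClasp character's face accessory, adding physics jiggle to ears, bows, jaw and tongue.
# It copies the attached parent's pose, offsets both ear bases additively, then chains AnimDynamics springs, with the tongue simulated relative to the jaw.
# The nose dynamics node is unconnected.
# EventGraph derives an ears-up value from smoothed skydive floating and velocity, but AnimGraph shows no use of it.
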AnimBlueprint /BRCosmetics/Characters/Player/Male/Medium/Bodies/M_MED_QuarterClasp_Birch/Meshes/Parts/M_MED_QuarterClasp_Birch_FaceAcc_AnimBP
target_skeleton=&/M_MED_QuarterClasp_Birch_FaceAcc_Skeleton
graphs [
 AnimGraph {nodes [
   Anim Dynamics - Bone: dyn_nose(Alpha=1, ExternalForce="0,0,0") {
    angularDampingOverride=0.9999
    angularSpringConstant=3000
    bAngularSpring=true
    bOverrideAngularDamping=true
    bUseGravityOverride=true
    linearDampingOverride=1
    linearSpringConstant=50
    simulationSpace=RootRelative
    alphaBoolBlend {blendOption=Linear}
    boundBone {boneName=dyn_nose}
    componentAppliedLinearAccClamp=<1000 1000 1000>
    componentLinearAccScale=<0.1 0.1 0.1>
    componentLinearVelScale=<0.1 0.1 0.1>
    gravityOverride {z=-980}
    physicsBodyDefinitions [{
      boundBone {boneName=dyn_nose}
      boxExtents=<10 10 10>
      constraintSetup {angularLimitsMax=<5 10 10>, angularLimitsMin=<-5 -10 -10>, angularTarget=<1 0 0>}
      localJointOffset=<0.1 0 0>
     }]
   }
   n1 = Local To Component(LocalPose=Copy Pose From Mesh() {bUseAttachedParent=true})
   n2 = Anim Dynamics - Bone: dyn_ear_l_01(
    ComponentPose="Transform (Modify) Bone - Bone: dyn_ear_r_base"(
     ComponentPose="Transform (Modify) Bone - Bone: dyn_ear_l_base"(ComponentPose=@n1, Alpha=1, Translation="0,0,0", Rotation="80,0,0", Scale="1,1,1") {rotationMode=BMM_Additive, boneToModify {boneName=dyn_ear_l_base}, rotation {pitch=80}}
     Alpha=1
     Translation="0,0,0"
     Rotation="-80,0,0"
     Scale="1,1,1"
    ) {rotationMode=BMM_Additive, boneToModify {boneName=dyn_ear_r_base}, rotation {pitch=-80}}
    Alpha=1
    ExternalForce="0,0,0"
   ) {
    angularDampingOverride=0.99
    angularSpringConstant=1000
    bAngularSpring=true
    bOverrideAngularDamping=true
    bUseGravityOverride=true
    linearDampingOverride=1
    linearSpringConstant=50
    simulationSpace=RootRelative
    alphaBoolBlend {blendOption=Linear}
    boundBone {boneName=dyn_ear_l_01}
    componentAppliedLinearAccClamp=<5000 5000 5000>
    componentLinearAccScale=<1 1 1>
    componentLinearVelScale=<1 1 1>
    gravityOverride {z=-980}
    physicsBodyDefinitions [{
      boundBone {boneName=dyn_ear_l_01}
      boxExtents=<10 10 10>
      constraintSetup {angularLimitsMax=<5 0 0>, angularLimitsMin=<-5 -45 -10>, angularTarget=<1 0 0>}
      localJointOffset=<0.3 0 0>
     }]
   }
   n3 = Anim Dynamics - Bone: dyn_ear_l_02(ComponentPose=@n2, Alpha=1, ExternalForce="0,0,0") {
    angularDampingOverride=0.99
    angularSpringConstant=1000
    bAngularSpring=true
    bOverrideAngularDamping=true
    bUseGravityOverride=true
    linearDampingOverride=1
    linearSpringConstant=50
    simulationSpace=RootRelative
    alphaBoolBlend {blendOption=Linear}
    boundBone {boneName=dyn_ear_l_02}
    componentAppliedLinearAccClamp=<5000 5000 5000>
    componentLinearAccScale=<1 1 1>
    componentLinearVelScale=<1 1 1>
    gravityOverride {z=-980}
    physicsBodyDefinitions [{
      boundBone {boneName=dyn_ear_l_02}
      boxExtents=<10 10 10>
      constraintSetup {angularLimitsMax=<15 3 10>, angularLimitsMin=<-15 -45 -10>, angularTarget=<1 0 0>}
      localJointOffset=<0.3 0 0>
     }]
   }
   n4 = Anim Dynamics - Bone: dyn_ear_r_01(ComponentPose=@n3, Alpha=1, ExternalForce="0,0,0") {
    angularDampingOverride=0.99
    angularSpringConstant=1000
    bAngularSpring=true
    bOverrideAngularDamping=true
    bUseGravityOverride=true
    linearDampingOverride=1
    linearSpringConstant=50
    simulationSpace=RootRelative
    alphaBoolBlend {blendOption=Linear}
    boundBone {boneName=dyn_ear_r_01}
    componentAppliedLinearAccClamp=<5000 5000 5000>
    componentLinearAccScale=<1 1 1>
    componentLinearVelScale=<1 1 1>
    gravityOverride {z=-980}
    physicsBodyDefinitions [{
      boundBone {boneName=dyn_ear_r_01}
      boxExtents=<10 10 10>
      constraintSetup {angularLimitsMax=<5 0 0>, angularLimitsMin=<-5 -45 -10>, angularTarget=<1 0 0>}
      localJointOffset=<0.4 0 0>
     }]
   }
   n5 = Anim Dynamics - Bone: dyn_ear_r_02(ComponentPose=@n4, Alpha=1, ExternalForce="0,0,0") {
    angularDampingOverride=0.99
    angularSpringConstant=1000
    bAngularSpring=true
    bOverrideAngularDamping=true
    bUseGravityOverride=true
    linearDampingOverride=1
    linearSpringConstant=50
    simulationSpace=RootRelative
    alphaBoolBlend {blendOption=Linear}
    boundBone {boneName=dyn_ear_r_02}
    componentAppliedLinearAccClamp=<5000 5000 5000>
    componentLinearAccScale=<1 1 1>
    componentLinearVelScale=<1 1 1>
    gravityOverride {z=-980}
    physicsBodyDefinitions [{
      boundBone {boneName=dyn_ear_r_02}
      boxExtents=<10 10 10>
      constraintSetup {angularLimitsMax=<15 3 10>, angularLimitsMin=<-15 -45 -10>, angularTarget=<1 0 0>}
      localJointOffset=<0.4 0 0>
     }]
   }
   n6 = Anim Dynamics - Bone: dyn_bow_l(ComponentPose=@n5, Alpha=1, ExternalForce="0,0,0") {
    angularDampingOverride=0.99
    angularSpringConstant=100
    bAngularSpring=true
    bOverrideAngularDamping=true
    bUseGravityOverride=true
    linearDampingOverride=1
    linearSpringConstant=50
    simulationSpace=RootRelative
    alphaBoolBlend {blendOption=Linear}
    boundBone {boneName=dyn_bow_l}
    componentAppliedLinearAccClamp=<3000 3000 3000>
    componentLinearAccScale=<1 1 1>
    componentLinearVelScale=<0.5 0.5 0.5>
    gravityOverride {z=-980}
    physicsBodyDefinitions [{
      boundBone {boneName=dyn_bow_l}
      boxExtents=<10 10 10>
      constraintSetup {angularLimitsMax=<5 3 10>, angularLimitsMin=<-5 -30 -3>, angularTarget=<1 0 0>}
      localJointOffset=<0.1 0 0>
     }]
   }
   n7 = Anim Dynamics - Bone: dyn_bow_r(ComponentPose=@n6, Alpha=1, ExternalForce="0,0,0") {
    angularDampingOverride=0.99
    angularSpringConstant=100
    bAngularSpring=true
    bOverrideAngularDamping=true
    bUseGravityOverride=true
    linearDampingOverride=1
    linearSpringConstant=50
    simulationSpace=RootRelative
    alphaBoolBlend {blendOption=Linear}
    boundBone {boneName=dyn_bow_r}
    componentAppliedLinearAccClamp=<3000 3000 3000>
    componentLinearAccScale=<1 1 1>
    componentLinearVelScale=<0.5 0.5 0.5>
    gravityOverride {z=-980}
    physicsBodyDefinitions [{
      boundBone {boneName=dyn_bow_r}
      boxExtents=<10 10 10>
      constraintSetup {angularLimitsMax=<5 3 10>, angularLimitsMin=<-5 -30 -3>, angularTarget=<1 0 0>}
      localJointOffset=<0.1 0 0>
     }]
   }
   n8 = Anim Dynamics - Bone: dyn_jaw(ComponentPose=@n7, Alpha=1, ExternalForce="0,0,0") {
    angularDampingOverride=1
    angularSpringConstant=2500
    bAngularSpring=true
    bOverrideAngularDamping=true
    bUseGravityOverride=true
    linearDampingOverride=1
    linearSpringConstant=50
    simulationSpace=RootRelative
    alphaBoolBlend {blendOption=Linear}
    boundBone {boneName=dyn_jaw}
    componentAppliedLinearAccClamp=<500 500 500>
    gravityOverride {z=-980}
    physicsBodyDefinitions [{
      boundBone {boneName=dyn_jaw}
      boxExtents=<11 11 11>
      constraintSetup {angularLimitsMax=<2 3 2>, angularLimitsMin=<-2 -8 -2>, angularTarget=<1 0 0>}
      localJointOffset=<0.1 0 0>
     }]
   }
   n9 = Anim Dynamics - Bone: dyn_tongue(ComponentPose=@n8, Alpha=1, ExternalForce="0,0,0") {
    angularDampingOverride=0.999
    angularSpringConstant=500
    bAngularSpring=true
    bOverrideAngularDamping=true
    bUseGravityOverride=true
    linearDampingOverride=1
    linearSpringConstant=50
    simulationSpace=BoneRelative
    alphaBoolBlend {blendOption=Linear}
    boundBone {boneName=dyn_tongue}
    componentAppliedLinearAccClamp=<5000 5000 5000>
    componentLinearAccScale=<1 1 1>
    componentLinearVelScale=<1 1 1>
    gravityOverride {z=-980}
    physicsBodyDefinitions [{
      boundBone {boneName=dyn_tongue}
      boxExtents=<10 10 10>
      constraintSetup {angularLimitsMax=<5 75 50>, angularLimitsMin=<-5 -95 -50>, angularTarget=<1 0 0>}
      localJointOffset=<1 0 0>
     }]
    relativeSpaceBone {boneName=dyn_jaw}
   }
   Output Pose(Result=Component To Local(ComponentPose=@n9)) {layerGroup=None}
  ]}
 EventGraph {nodes [
   n1 = Event BlueprintUpdateAnimation()
   n2 = Get bIsSkydiveFloating()
   > n3 = Set is Skydiving(is Skydiving:double=FInterpTo(Current=Get is Skydiving() -> (is Skydiving:double), Target="To Float (Boolean)"(InBool=@n2), DeltaTime=@n1.DeltaTimeX, InterpSpeed=5)) -> (Output_Get:double)
   n4 = Vector Length(A=Get PawnVelocity())
   > Set is Ears Up(is Ears Up:double=float * float(A=Remap Float(Input=@n4, OldRange=(X=7000,Y=2000), NewRange=(X=0,Y=1), Clamp=true) -> (Output=~), B=@n3) {is=title K2Node_PromotableOperator}) -> (Output_Get:double)
  ]}
]
variables [
 is Ears Up {type=number}
 is Skydiving {type=number}
]
